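# Sound cue for the closing sound of the Bootleg faction door A, using the Fort_Doors sound class and shared door attenuation.
# A mixer layers a random choice of two BurnTyres double-door closes with a metal door close delayed by a fixed quarter second.
# A random choice of two metal bulkhead door closes is mixed underneath at reduced volume.
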SoundCue /Hermes/Sets/Bootleg/Bootleg_FactionDoor_A/Audio/SC_Bootleg_FactionDoor_Close
root=Mixer_0
nodes [
 {id=Delay_8, children [WavePlayer_5], properties {delayMax=0.25, delayMin=0.25}}
 {id=Mixer_0, children [Random_2,Delay_8,Random_0], properties {inputVolume [1,1,0.25]}}
 {id=Random_0, children [WavePlayer_6,WavePlayer_7], properties {weights [1,1]}}
 {id=Random_2, children [WavePlayer_10, WavePlayer_11], properties {weights [1,1]}}
 {id=WavePlayer_10, properties {soundWaveAssetPtr=&/BurnTyres_DoubleDoor_Close_01}}
 {id=WavePlayer_11, properties {soundWaveAssetPtr=&/BurnTyres_DoubleDoor_Close_03}}
 {id=WavePlayer_5, properties {soundWaveAssetPtr=&/metal_door_close_C}}
 {id=WavePlayer_6, properties {soundWaveAssetPtr=&/Metal_Bulkhead_Door_Close_02}}
 {id=WavePlayer_7, properties {soundWaveAssetPtr=&/Metal_Bulkhead_Door_Close_01}}
]
properties {
 attenuationSettings=/Game/Sounds/Attenuation/Fort_Doors_Att
 maxDistance=2400
 soundClassObject=&Fort_Doors
 volumeMultiplier=0.9
}
