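# Sound cue for opening the Bootleg Faction Door A in the Hermes set.
# A mixer layers a ship hatch open sound at full volume with a double metal door open sound at quarter volume.
# A random node adds one of two equally weighted metal door open variations at full volume, so repeated openings vary.
# It uses the Fort doors sound class and shared door attenuation.
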SoundCue /Hermes/Sets/Bootleg/Bootleg_FactionDoor_A/Audio/SC_Bootleg_FactionDoor_Open
root=Mixer_0
nodes [
 {id=Mixer_0, children [WavePlayer_9,Random_0,WavePlayer_6], properties {inputVolume [1,1,0.25]}}
 {id=Random_0, children [WavePlayer_8,WavePlayer_7], properties {weights [1,1]}}
 {id=WavePlayer_6, properties {soundWaveAssetPtr=&/Metal_DoubleDoor_B_A_Open_01}}
 {id=WavePlayer_7, properties {soundWaveAssetPtr=&/metal_door_open_B}}
 {id=WavePlayer_8, properties {soundWaveAssetPtr=&/metal_door_open_A}}
 {id=WavePlayer_9, properties {soundWaveAssetPtr=&/Door_Metal_Ship_Hatch_Open_01}}
]
properties {
 attenuationSettings=/Game/Sounds/Attenuation/Fort_Doors_Att
 maxDistance=2400
 soundClassObject=&Fort_Doors
}
