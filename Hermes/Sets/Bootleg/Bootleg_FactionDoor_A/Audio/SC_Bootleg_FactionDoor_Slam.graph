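# Sound cue for the slam of the Bootleg faction door A.
# A mixer layers a fixed double-door slam with one of four metal bulkhead door slams, picked at random with equal weight for variation.
# It plays through the shared door sound class and door attenuation settings.
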SoundCue /Hermes/Sets/Bootleg/Bootleg_FactionDoor_A/Audio/SC_Bootleg_FactionDoor_Slam
root=Mixer_0
nodes [
 {id=Mixer_0, children [WavePlayer_5,Random_2], properties {inputVolume [1,1]}}
 {id=Random_2, children [WavePlayer_6,WavePlayer_7,WavePlayer_8,WavePlayer_9], properties {weights [1,1,1,1]}}
 {id=WavePlayer_5, properties {soundWaveAssetPtr=/Game/Sounds/Fort_Doors/World/BurnTyres_DoubleDoor_Slam_01}}
 {id=WavePlayer_6, properties {soundWaveAssetPtr=/Game/Sounds/Fort_Doors/World/Metal_Bulkhead_Door_Slam_01}}
 {id=WavePlayer_7, properties {soundWaveAssetPtr=/Game/Sounds/Fort_Doors/World/Metal_Bulkhead_Door_Slam_02}}
 {id=WavePlayer_8, properties {soundWaveAssetPtr=/Game/Sounds/Fort_Doors/World/Metal_Bulkhead_Door_Slam_03}}
 {id=WavePlayer_9, properties {soundWaveAssetPtr=/Game/Sounds/Fort_Doors/World/Metal_Bulkhead_Door_Slam_04}}
]
properties {
 attenuationSettings=/Game/Sounds/Attenuation/Fort_Doors_Att
 maxDistance=2400
 soundClassObject=&Fort_Doors
 volumeMultiplier=1
}
